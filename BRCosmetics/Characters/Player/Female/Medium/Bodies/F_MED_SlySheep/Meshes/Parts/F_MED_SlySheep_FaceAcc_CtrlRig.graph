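# Control Rig for the SlySheep female character's face accessory mesh, applying head-pose-driven corrective offsets to back curl and neck bones.
# A sequence runs spherical pose readers on the head: one offsets and squashes the right back-curl root, another has no consumer.
# Two others drive rotate-and-scale offsets on the left and right extra neck bones, gated by FE and not Emote.
# Aim and parent-constraint chains for the curl roots are left unconnected to execution.
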ControlRigBlueprint /BRCosmetics/Characters/Player/Female/Medium/Bodies/F_MED_SlySheep/Meshes/Parts/F_MED_SlySheep_FaceAcc_CtrlRig
graphs [
 RigVMFunctionLibrary {links [], nodes []}
 RigVMFunction_Sequence_1_ContainedGraph {nodes [
   n1 = RigVMFunctionEntryNode()
   RigVMFunction_Sequence = &RigVMFunction_Sequence(ExecuteContext=@n1.ExecuteContext)
   n2 = &RigVMFunction_Sequence(ExecuteContext=@RigVMFunction_Sequence.B)
   n3 = &RigVMFunction_Sequence(ExecuteContext=@n2.B)
   RigVMFunctionReturnNode(A=@RigVMFunction_Sequence.A, B=@n2.A, C=@n3.A, D=@n3.B)
  ]}
 RigVMModel {nodes [
   AlphaInterp = &RigVMFunction_AlphaInterp(InterpSpeedDecreasing=6, InterpSpeedIncreasing=6, Scale=1, Value=0, bClampResult=False)
   /Script/ControlRig.RigUnit_AimItem(
    DebugSettings=(bEnabled=True,Scale=10,WorldOffset=(Rotation=(X=0,Y=0,Z=0,W=1),Translation=(X=0,Y=0,Z=0),Scale3D=(X=1,Y=1,Z=1)))
    ExecutePin=()
    Item=(Type=Bone,Name="dyn_root_curl_bk")
    Primary=(Weight=1,Axis=(X=0,Y=0,Z=-1),Target=(X=35.5615,Y=-9.0943,Z=4.346),Kind=Location,Space=(Type=Bone,Name="pelvis"))
    Secondary=(Weight=1,Axis=(X=0,Y=-1,Z=0),Target=(X=20.8256,Y=-3.7921,Z=17.3989),Kind=Location,Space=(Type=Bone,Name="spine_05"))
    Weight=@AlphaInterp.Result
   ) {is=struct RigVMUnitNode}
   > &RigUnit_AimItem(
    DebugSettings=(bEnabled=False,Scale=10,WorldOffset=(Rotation=(X=0,Y=0,Z=0,W=1),Translation=(X=0,Y=0,Z=0),Scale3D=(X=1,Y=1,Z=1)))
    Item=(Type=Bone,Name="dyn_root_curl_bk_r")
    Primary=(Weight=1,Axis=(X=0,Y=0,Z=-1),Target=(X=35.56,Y=-9.0333,Z=4.7191),Kind=Location,Space=(Type=Bone,Name="pelvis"))
    Secondary=(Weight=1,Axis=(X=0,Y=-1,Z=0),Target=(X=20.8256,Y=-3.7921,Z=17.3989),Kind=Location,Space=(Type=Bone,Name="spine_05"))
    Weight=@AlphaInterp.Result
   )
   &RigUnit_ParentConstraint(
    AdvancedSettings=(InterpolationType=Average,RotationOrderForFilter=XZY)
    Child=(Type=Bone,Name="dyn_root_curl_bk")
    ExecutePin=()
    Filter=(TranslationFilter=(bX=False,bY=False,bZ=False),RotationFilter=(bX=True,bY=True,bZ=True),ScaleFilter=(bX=True,bY=True,bZ=True))
    Parents=((Item=(Type=Bone,Name="spine_05"),Weight=1))
    bMaintainOffset=true
    Weight=@AlphaInterp.Result
   )
   > &RigUnit_OffsetTransformForItem(
    Item=(Type=Bone,Name="dyn_root_curl_bk")
    OffsetTransform=(Rotation=(X=0,Y=0,Z=0,W=1),Translation=(X=-2,Y=0,Z=2),Scale3D=(X=1,Y=1,Z=1))
    bPropagateToChildren=True
    Weight=@AlphaInterp.Result
   )
   n1 = &RigVMFunction_AlphaInterp(
    InterpSpeedDecreasing=6
    InterpSpeedIncreasing=6
    Scale=0.5
    bClampResult=False
    bInterpResult=false
    bMapRange=true
    Value=@AlphaInterp.Result
   )
   > &RigUnit_ParentConstraint(
    AdvancedSettings=(InterpolationType=Average,RotationOrderForFilter=XZY)
    Child=(Type=Bone,Name="dyn_root_curl_bk_r")
    Filter=(TranslationFilter=(bX=False,bY=False,bZ=False),RotationFilter=(bX=True,bY=True,bZ=True),ScaleFilter=(bX=True,bY=True,bZ=True))
    Parents=((Item=(Type=Bone,Name="spine_05"),Weight=1))
    bMaintainOffset=true
    Weight=@n1.Result
   )
   > &RigUnit_OffsetTransformForItem(
    Item=(Type=Bone,Name="dyn_root_curl_bk_r")
    OffsetTransform=(Rotation=(X=0.0871,Y=0.0435,Z=-0.0038,W=0.9952),Translation=(X=2,Y=0,Z=3),Scale3D=(X=1,Y=1,Z=0.8))
    bPropagateToChildren=True
    Weight=@AlphaInterp.Result
   )
   > &RigUnit_ParentConstraint(
    AdvancedSettings=(InterpolationType=Average,RotationOrderForFilter=XZY)
    Child=(Type=Bone,Name="dyn_root_curl_bk_l")
    Filter=(TranslationFilter=(bX=False,bY=False,bZ=False),RotationFilter=(bX=True,bY=True,bZ=True),ScaleFilter=(bX=True,bY=True,bZ=True))
    Parents=((Item=(Type=Bone,Name="spine_05"),Weight=1))
    bMaintainOffset=true
    Weight=@n1.Result
   )
   > &RigUnit_OffsetTransformForItem(
    Item=(Type=Bone,Name="dyn_root_curl_bk_l")
    OffsetTransform=(Rotation=(X=0,Y=0.0436,Z=0,W=0.999),Translation=(X=2,Y=0,Z=3),Scale3D=(X=1,Y=1,Z=0.8))
    bPropagateToChildren=True
    Weight=@AlphaInterp.Result
   )
   RigUnit_BeginExecution = &RigUnit_BeginExecution()
   RigVMFunction_Sequence = RigVMAggregateNode(ExecuteContext=@RigUnit_BeginExecution.ExecutePin)
   &RigUnit_SphericalPoseReader(
    ActiveRegionScaleFactors=(PositiveWidth=0.5,NegativeWidth=1,PositiveHeight=0.75,NegativeHeight=0.75)
    ActiveRegionSize=0.1
    Debug=(bDrawDebug=True,bDraw2D=False,bDrawLocalAxes=False,DebugScale=25,DebugSegments=20,DebugThickness=0.25)
    DriverAxis=(X=0,Y=1,Z=0)
    DriverItem=(Type=Bone,Name="head")
    FalloffRegionScaleFactors=(PositiveWidth=0.5,NegativeWidth=1,PositiveHeight=1,NegativeHeight=1)
    FalloffSize=0.2
    FlipHeightScaling=False
    FlipWidthScaling=False
    OptionalParentItem=(Type=Bone,Name="None")
    RotationOffset=(X=90,Y=0,Z=30)
    ExecutePin=@RigVMFunction_Sequence.B
   )
   n2 = &RigUnit_SphericalPoseReader(
    ActiveRegionScaleFactors=(PositiveWidth=0.5,NegativeWidth=0.1,PositiveHeight=1,NegativeHeight=1)
    ActiveRegionSize=0.2
    Debug=(bDrawDebug=True,bDraw2D=False,bDrawLocalAxes=False,DebugScale=25,DebugSegments=20,DebugThickness=0.25)
    DriverAxis=(X=0,Y=1,Z=0)
    DriverItem=(Type=Bone,Name="head")
    FalloffRegionScaleFactors=(PositiveWidth=1,NegativeWidth=0.2,PositiveHeight=1,NegativeHeight=1)
    FalloffSize=0.4
    FlipHeightScaling=False
    FlipWidthScaling=False
    OptionalParentItem=(Type=Bone,Name="None")
    RotationOffset=(X=90,Y=0,Z=-30)
    ExecutePin=@RigVMFunction_Sequence.A
   )
   > &RigUnit_OffsetTransformForItem(
    Item=(Type=Bone,Name="dyn_root_curl_bk_r")
    OffsetTransform=(Rotation=(X=-0.0436,Y=0,Z=0,W=0.999),Translation=(X=-2,Y=0,Z=2),Scale3D=(X=1,Y=1,Z=0.8))
    bPropagateToChildren=True
    Weight=@n2.OutputParam
   )
   n3 = &RigUnit_SphericalPoseReader(
    ActiveRegionScaleFactors=(PositiveWidth=0.75,NegativeWidth=0.75,PositiveHeight=0.5,NegativeHeight=1)
    ActiveRegionSize=0.15
    Debug=(bDrawDebug=True,bDraw2D=False,bDrawLocalAxes=False,DebugScale=25,DebugSegments=20,DebugThickness=0.25)
    DriverAxis=(X=0,Y=1,Z=0)
    DriverItem=(Type=Bone,Name="head")
    FalloffRegionScaleFactors=(PositiveWidth=1,NegativeWidth=1,PositiveHeight=0.6968,NegativeHeight=1)
    FalloffSize=0.2
    FlipHeightScaling=False
    FlipWidthScaling=False
    OptionalParentItem=(Type=Bone,Name="neck_02")
    RotationOffset=(X=15,Y=0,Z=0)
    ExecutePin=@RigVMFunction_Sequence.D
   )
   RigVMFunction_MathBoolNot = &RigVMFunction_MathBoolNot(Value=RigVMVariableNode() {variable=Emote}.Value)
   RigVMFunction_MathBoolAnd = &RigVMFunction_MathBoolAnd(B=@RigVMFunction_MathBoolNot.Result, A=RigVMVariableNode() {variable=FE}.Value)
   RigVMFunction_MathBoolToFloat = &RigVMFunction_MathBoolToFloat(Value=@RigVMFunction_MathBoolAnd.Result)
   n4 = &RigVMFunction_AlphaInterp(
    InterpSpeedDecreasing=10
    InterpSpeedIncreasing=10
    Scale=1
    bClampResult=False
    bInterpResult=False
    Value=&RigVMFunction_MathDoubleMul(A=@RigVMFunction_MathBoolToFloat.Result, B=@n3.OutputParam).Result
   )
   > &RigUnit_OffsetTransformForItem(
    Item=(Type=Bone,Name="extra_neck_r")
    OffsetTransform=(Rotation=(X=0,Y=0.4226,Z=0,W=0.9063),Translation=(X=0,Y=0,Z=0),Scale3D=(X=1,Y=1,Z=0.8))
    bPropagateToChildren=True
    Weight=@n4.Result
   )
   n5 = &RigUnit_SphericalPoseReader(
    ActiveRegionScaleFactors=(PositiveWidth=0.75,NegativeWidth=0.75,PositiveHeight=1,NegativeHeight=0.5)
    ActiveRegionSize=0.05
    Debug=(bDrawDebug=True,bDraw2D=False,bDrawLocalAxes=False,DebugScale=25,DebugSegments=20,DebugThickness=0.25)
    DriverAxis=(X=0,Y=1,Z=0)
    DriverItem=(Type=Bone,Name="head")
    FalloffRegionScaleFactors=(PositiveWidth=1,NegativeWidth=1,PositiveHeight=1,NegativeHeight=1)
    FalloffSize=0.3
    FlipHeightScaling=False
    FlipWidthScaling=False
    OptionalParentItem=(Type=Bone,Name="root")
    RotationOffset=(X=80,Y=0,Z=0)
    ExecutePin=@RigVMFunction_Sequence.C
   )
   Multiply = &RigVMFunction_MathDoubleMul(B=@RigVMFunction_MathBoolToFloat.Result, A=@n5.OutputParam)
   n6 = &RigVMFunction_AlphaInterp(InterpSpeedDecreasing=10, InterpSpeedIncreasing=10, Scale=1, bClampResult=False, bInterpResult=False, Value=@Multiply.Result)
   > &RigUnit_OffsetTransformForItem(
    Item=(Type=Bone,Name="extra_neck_l")
    OffsetTransform=(Rotation=(X=0,Y=-0.2588,Z=0,W=0.9659),Translation=(X=0,Y=0,Z=0),Scale3D=(X=1,Y=1,Z=0.8))
    bPropagateToChildren=True
    Weight=@n6.Result
   )
   &RigUnit_AimItem(
    DebugSettings=(bEnabled=True,Scale=10,WorldOffset=(Rotation=(X=0,Y=0,Z=0,W=1),Translation=(X=0,Y=0,Z=0),Scale3D=(X=1,Y=1,Z=1)))
    ExecutePin=()
    Item=(Type=Bone,Name="dyn_root_curl_bk_r")
    Primary=(Weight=1,Axis=(X=0,Y=0,Z=-1),Target=(X=35.5615,Y=-9.0733,Z=6.6838),Kind=Location,Space=(Type=Bone,Name="pelvis"))
    Secondary=(Weight=1,Axis=(X=0,Y=-1,Z=0),Target=(X=20.8256,Y=-3.7921,Z=17.3989),Kind=Location,Space=(Type=Bone,Name="spine_05"))
    Weight=@AlphaInterp.Result
   )
  ]}
]
hierarchy [
 root {loc=<0 0 0>, rot=<0 0 0 1>}
 pelvis {parent=root, loc=<0 -0.5301 98.0571>, rot=<0.0127 0.707 -0.0127 -0.707>}
 spine_01 {parent=pelvis, loc=<6.1208 -0.9328 0>, rot=<0 0 0.034 -0.9994>}
 spine_02 {parent=spine_01, loc=<6.5997 -0.2095 0>, rot=<0 0 0.0238 -0.9997>}
 spine_03 {parent=spine_02, loc=<6.4914 -0.1237 0>, rot=<0 0 0.0149 -0.9999>}
 spine_04 {parent=spine_03, loc=<6.9598 0.0794 0>, rot=<0 0 -0.0132 -0.9999>}
 spine_05 {parent=spine_04, loc=<6.9987 0 0>, rot=<0 0 -0.0015 -1>}
 neck_01 {parent=spine_05, loc=<11.3138 0.1529 0>, rot=<0 0 -0.2003 -0.9797>}
 neck_02 {parent=neck_01, loc=<4.5627 0 0>, rot=<0 0 0.0504 -0.9987>}
 head {parent=neck_02, loc=<5.3914 0 0>, rot=<0 0 0.115 -0.9934>}
 dyn_curl_side_r_base {parent=head, loc=<7.7895 4.7949 8.6493>, rot=<0.0159 0.1948 0.9805 -0.0186>}
 dyn_curl_side_r_01 {parent=dyn_curl_side_r_base, loc=<0.2 0 0>, rot=<0 0 0 1>}
 dyn_curl_side_r_02 {parent=dyn_curl_side_r_01, loc=<6.12 -0.4174 -0.3247>, rot=<0.0996 -0.0658 -0.1002 -0.9878>}
 dyn_curl_fr_r_base {parent=head, loc=<10.8905 9.0773 3.6831>, rot=<0.0706 0.5901 0.8004 0.0785>}
 dyn_curl_fr_r_01 {parent=dyn_curl_fr_r_base, loc=<0.2 0 0>, rot=<0 0 0 1>}
 dyn_curl_fr_r_02 {parent=dyn_curl_fr_r_01, loc=<6.3714 -0.1504 -0.4113>, rot=<0.006 -0.1112 0.0534 -0.9923>}
 dyn_curl_fr_side_r_base {parent=head, loc=<9.3725 7.6008 6.3782>, rot=<-0.0209 0.4281 0.9032 -0.0218>}
 dyn_curl_fr_side_r_01 {parent=dyn_curl_fr_side_r_base, loc=<0.2 0 0>, rot=<0 0 0 1>}
 dyn_curl_fr_side_02 {parent=dyn_curl_fr_side_r_01, loc=<6.6624 -0.225 0.3596>, rot=<0 0 0 1>}
 dyn_curl_fr_l_base {parent=head, loc=<11.8757 8.9611 0.456>, rot=<-0.1163 -0.7234 -0.6567 -0.1785>}
 dyn_curl_fr_l_01 {parent=dyn_curl_fr_l_base, loc=<0.2 0 0>, rot=<0 0 0 1>}
 dyn_curl_fr_l_02 {parent=dyn_curl_fr_l_01, loc=<3.953 0.1212 -0.2407>, rot=<0.0121 -0.1107 -0.0016 -0.9938>}
 dyn_curl_fr_side_l_01_base {parent=head, loc=<11.2263 8.1793 -2.8748>, rot=<-0.1477 -0.8003 -0.5599 -0.1555>}
 dyn_curl_fr_side_l_01 {parent=dyn_curl_fr_side_l_01_base, loc=<0.2 0 0>, rot=<0 0 0 1>}
 dyn_curl_fr_side_l_02_base {parent=head, loc=<9.422 7.2317 -4.9595>, rot=<-0.078 -0.8133 -0.5738 -0.0562>}
 dyn_curl_fr_side_l_02 {parent=dyn_curl_fr_side_l_02_base, loc=<0.2 0 0>, rot=<0 0 0 1>}
 dyn_curl_side_l_base {parent=head, loc=<7.1017 4.1463 -8.5068>, rot=<-0.0023 -0.9759 -0.2177 0.0119>}
 dyn_curl_side_l_01 {parent=dyn_curl_side_l_base, loc=<0.2 0 0>, rot=<0 0 0 1>}
 dyn_curl_side_l_02 {parent=dyn_curl_side_l_01, loc=<5.6365 0 0>, rot=<0.0064 -0.0504 0.1264 -0.9907>}
 dyn_ear_r_base {parent=head, loc=<1.9173 0.5932 6.8505>, rot=<0.3443 0.0179 0.927 -0.1475>}
 dyn_ear_r {parent=dyn_ear_r_base, loc=<0.2 0 0>, rot=<0 0 0 1>}
 dyn_ear_l_base {parent=head, loc=<1.9169 0.5932 -6.8505>, rot=<0.1475 -0.927 -0.0179 -0.3443>}
 dyn_ear_l {parent=dyn_ear_l_base, loc=<0.2 0 0>, rot=<0 0 0 1>}
 dyn_curl_side_bk_l_base {parent=head, loc=<3.7508 -1.1101 -8.8913>, rot=<0.2223 -0.8141 0.1142 -0.5243>}
 dyn_curl_side_bk_l {parent=dyn_curl_side_bk_l_base, loc=<0.2 0 0>, rot=<0 0 0 1>}
 dyn_curl_side_bk_r_base {parent=head, loc=<5.4385 -0.431 8.9168>, rot=<0.3414 0.1396 0.9259 0.0818>}
 dyn_curl_side_bk_r {parent=dyn_curl_side_bk_r_base, loc=<0.2 0 0>, rot=<0 0 0 1>}
 dyn_root_curl_bk {parent=head, loc=<4.8868 -3.5177 0>, rot=<0.5021 -0.4979 0.4979 -0.5021>}
 dyn_curl_bk_03_base {parent=dyn_root_curl_bk, loc=<4.8592 2.813 -1.4582>, rot=<0.0944 -0.0764 -0.4303 -0.8945>}
 dyn_curl_bk_03 {parent=dyn_curl_bk_03_base, loc=<0.2 0 0>, rot=<0 0 0 1>}
 dyn_curl_bk_04_base {parent=dyn_root_curl_bk, loc=<5.6696 3.0106 -9.5337>, rot=<-0.019 -0.5442 -0.2837 -0.7893>}
 dyn_curl_bk_04 {parent=dyn_curl_bk_04_base, loc=<0.2 0 0>, rot=<0 0 0 1>}
 dyn_curl_bk_05_base {parent=dyn_root_curl_bk, loc=<5.7629 -0.031 -8.5842>, rot=<-0.0306 -0.527 0.1453 -0.8368>}
 dyn_curl_bk_05 {parent=dyn_curl_bk_05_base, loc=<0.2 0 0>, rot=<0 0 0 1>}
 dyn_curl_bk_06_base {parent=dyn_root_curl_bk, loc=<4.6919 -3.7792 -7.4225>, rot=<0.449 -0.4845 -0.2949 -0.6904>}
 dyn_curl_bk_06 {parent=dyn_curl_bk_06_base, loc=<0.2 0 0>, rot=<0 0 0 1>}
 extra_mask {parent=head, loc=<0 8.6875 0>, rot=<0 0 0 1>}
 dyn_root_curl_bk_r {parent=head, loc=<4 -3.5 6>, rot=<0.5021 -0.4979 0.4979 -0.5021>}
 dyn_curl_bk_07_base {parent=dyn_root_curl_bk_r, loc=<1.9034 -0.9231 -8.5609>, rot=<-0.0546 -0.1913 0.0352 -0.9794>}
 dyn_curl_bk_07 {parent=dyn_curl_bk_07_base, loc=<0.2 0 0>, rot=<0 0 0 1>}
 dyn_curl_bk_08_base {parent=dyn_root_curl_bk_r, loc=<1.8247 -0.719 -7.3455>, rot=<0.0123 0.0424 0.2351 -0.971>}
 dyn_curl_bk_08 {parent=dyn_curl_bk_08_base, loc=<0.2 0 0>, rot=<0 0 0 1>}
 dyn_curl_bk_09_base {parent=dyn_root_curl_bk_r, loc=<1.27 -1.1417 -1.8558>, rot=<-0.1863 -0.2093 0.7851 -0.5523>}
 dyn_curl_bk_09 {parent=dyn_curl_bk_09_base, loc=<0.2 0 0>, rot=<0 0 0 1>}
 dyn_root_curl_bk_l {parent=head, loc=<4 -3.5 -4.2328>, rot=<0.5021 -0.4979 0.4979 -0.5021>}
 dyn_curl_bk_01_base {parent=dyn_root_curl_bk_l, loc=<2.6669 1.6243 2.3883>, rot=<0.0817 -0.0899 -0.5613 -0.8187>}
 dyn_curl_bk_01 {parent=dyn_curl_bk_01_base, loc=<0.2 0 0>, rot=<0 0 0 1>}
 dyn_curl_bk_02_base {parent=dyn_root_curl_bk_l, loc=<3.745 0.2839 -6.4467>, rot=<0.1014 -0.1133 -0.5295 -0.8346>}
 dyn_curl_bk_02 {parent=dyn_curl_bk_02_base, loc=<0.2 0 0>, rot=<0 0 0 1>}
 clavicle_l {parent=spine_05, loc=<5.4831 -0.1053 -3.2946>, rot=<0.0465 -0.7289 -0.009 -0.6829>}
 upperarm_l {parent=clavicle_l, loc=<9.8658 -1.2136 -0.4105>, rot=<0.0164 -0.3245 -0.0269 -0.9454>}
 clavicle_r {parent=spine_05, loc=<5.4831 -0.1053 3.2946>, rot=<0.7289 0.0465 -0.6829 0.009>}
 upperarm_r {parent=clavicle_r, loc=<-9.8658 1.2136 0.4105>, rot=<0.0164 -0.3245 -0.0269 -0.9454>}
 extra_neck_l {parent=spine_05, loc=<7.3668 4.6116 -5.6683>, rot=<0.941 0.2251 -0.0425 -0.2491>}
 extra_neck_r {parent=spine_05, loc=<7.2579 4.8675 6.1647>, rot=<0.9408 0.2006 -0.0464 0.2691>}
]
variables [
 Emote {type=bool, default=False}
 FE {type=bool, default=False}
]
