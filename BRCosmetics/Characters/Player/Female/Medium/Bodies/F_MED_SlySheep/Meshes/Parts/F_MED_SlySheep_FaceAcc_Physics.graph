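# Physics asset for the SlySheep female medium character's face accessory part, covering dangling wool or hair curls and two ears for secondary motion.
# Capsules on the upper spine, neck, head, clavicles and upper arms provide body volumes around the small dyn_ capsules.
# Front and side curls and ears chain from base bones on the head, while back curls fan out from three root bodies, with per-joint swing and twist limits.
# Most base-to-head joints rotate freely, but the right ear and right back-side curl bases are locked.
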PhysicsAsset /BRCosmetics/Characters/Player/Female/Medium/Bodies/F_MED_SlySheep/Meshes/Parts/F_MED_SlySheep_FaceAcc_Physics
bodies {
 clavicle_l [{shapeType=Capsule, length=5, radius=7.5, shapeName=clavicle_l_capsule, center=<5 0 0>, rotation=<rot 0 90 -90>}]
 clavicle_r [{shapeType=Capsule, length=5, radius=7.5, shapeName=clavicle_r_capsule, center=<-5 0 0>, rotation=<rot 0 90 90>}]
 dyn_curl_bk_01 [{shapeType=Capsule, length=2, radius=1, shapeName=dyn_curl_bk_01_capsule, center=<2 0 0>, rotation=<rot 90 0 0>}]
 dyn_curl_bk_02 [{shapeType=Capsule, length=2, radius=1, shapeName=dyn_curl_bk_02_capsule, center=<2 0 0>, rotation=<rot 90 0 0>}]
 dyn_curl_bk_03 [{shapeType=Capsule, length=2, radius=1, shapeName=dyn_curl_bk_03_capsule, center=<2 0 0>, rotation=<rot 90 0 0>}]
 dyn_curl_bk_04 [{shapeType=Capsule, length=4, radius=1, shapeName=dyn_curl_bk_04_capsule, center=<4 0 0>, rotation=<rot 90 0 0>}]
 dyn_curl_bk_05 [{shapeType=Capsule, length=4, radius=1, shapeName=dyn_curl_bk_05_capsule, center=<4 0 0>, rotation=<rot 90 0 0>}]
 dyn_curl_bk_06 [{shapeType=Capsule, length=4, radius=1, shapeName=dyn_curl_bk_06_capsule, center=<4 0 0>, rotation=<rot 90 0 0>}]
 dyn_curl_bk_07 [{shapeType=Capsule, length=4, radius=2, shapeName=dyn_curl_bk_07_capsule, center=<4 0 0>, rotation=<rot 90 0 0>}]
 dyn_curl_bk_08 [{shapeType=Capsule, length=4, radius=1, shapeName=dyn_curl_bk_08_capsule, center=<4 0 0>, rotation=<rot 90 0 0>}]
 dyn_curl_bk_09 [{shapeType=Capsule, length=2, radius=1, shapeName=dyn_curl_bk_09_capsule, center=<2 0 0>, rotation=<rot 90 0 0>}]
 dyn_curl_fr_l_01 [{shapeType=Capsule, length=2, radius=2, shapeName=dyn_curl_fr_l_01_capsule, center=<2 0 0>, rotation=<rot 90 0 0>}]
 dyn_curl_fr_l_02 [{shapeType=Capsule, length=2, radius=2, shapeName=dyn_curl_fr_l_02_capsule, center=<2 0 0>, rotation=<rot 90 0 0>}]
 dyn_curl_fr_l_base [{shapeType=Capsule, length=1, radius=1, shapeName=dyn_curl_fr_l_base_capsule}]
 dyn_curl_fr_r_01 [{shapeType=Capsule, length=4, radius=2, shapeName=dyn_curl_fr_r_01_capsule, center=<4 0 0>, rotation=<rot 90 0 0>}]
 dyn_curl_fr_r_02 [{shapeType=Capsule, length=4, radius=2, shapeName=dyn_curl_fr_r_02_capsule, center=<4 0 0>, rotation=<rot 90 0 0>}]
 dyn_curl_fr_r_base [{shapeType=Capsule, length=1, radius=1, shapeName=dyn_curl_fr_r_base_capsule}]
 dyn_curl_fr_side_02 [{shapeType=Capsule, length=6, radius=2, shapeName=dyn_curl_fr_side_02_capsule, center=<6 0 0>, rotation=<rot 90 0 0>}]
 dyn_curl_fr_side_l_01 [{shapeType=Capsule, length=2, radius=2, shapeName=dyn_curl_fr_side_l_01_capsule, center=<2 0 0>, rotation=<rot 90 0 0>}]
 dyn_curl_fr_side_l_01_base [{shapeType=Capsule, length=1, radius=1, shapeName=dyn_curl_fr_side_l_01_base_capsule}]
 dyn_curl_fr_side_l_02 [{shapeType=Capsule, length=2, radius=2, shapeName=dyn_curl_fr_side_l_02_capsule, center=<2 0 0>, rotation=<rot 90 0 0>}]
 dyn_curl_fr_side_l_02_base [{shapeType=Capsule, length=1, radius=1, shapeName=dyn_curl_fr_side_l_02_base_capsule}]
 dyn_curl_fr_side_r_01 [{shapeType=Capsule, length=4, radius=1.5, shapeName=dyn_curl_fr_side_r_01_capsule, center=<4 0 0>, rotation=<rot 90 0 0>}]
 dyn_curl_fr_side_r_base [{shapeType=Capsule, length=1, radius=1, shapeName=dyn_curl_fr_side_r_base_capsule}]
 dyn_curl_side_bk_l [{shapeType=Capsule, length=2, radius=2, shapeName=dyn_curl_side_bk_l_capsule, center=<2 0 0>, rotation=<rot 90 0 0>}]
 dyn_curl_side_bk_l_base [{shapeType=Capsule, length=1, radius=1, shapeName=dyn_curl_side_bk_l_base_capsule}]
 dyn_curl_side_bk_r [{shapeType=Capsule, length=2, radius=2, shapeName=dyn_curl_side_bk_r_capsule, center=<4 0 0>, rotation=<rot 90 0 0>}]
 dyn_curl_side_bk_r_base [{shapeType=Capsule, length=1, radius=1, shapeName=dyn_curl_side_bk_r_base_capsule}]
 dyn_curl_side_l_01 [{shapeType=Capsule, length=2, radius=1.5, shapeName=dyn_curl_side_l_01_capsule, center=<2 0 0>, rotation=<rot 90 0 0>}]
 dyn_curl_side_l_02 [{shapeType=Capsule, length=6, radius=2, shapeName=dyn_curl_side_l_02_capsule, center=<4 0 0>, rotation=<rot 90 0 0>}]
 dyn_curl_side_l_base [{shapeType=Capsule, length=1, radius=1, shapeName=dyn_curl_side_l_base_capsule}]
 dyn_curl_side_r_01 [{shapeType=Capsule, length=2, radius=1.5, shapeName=dyn_curl_side_r_01_capsule, center=<2 0 0>, rotation=<rot 90 0.2804 0>}]
 dyn_curl_side_r_02 [{shapeType=Capsule, length=6, radius=2, shapeName=dyn_curl_side_r_02_capsule, center=<6 0 0>, rotation=<rot 90 0 0>}]
 dyn_curl_side_r_base [{shapeType=Capsule, length=1, radius=1, shapeName=dyn_curl_side_r_base_capsule}]
 dyn_ear_l [{shapeType=Capsule, length=6, radius=3, shapeName=dyn_ear_l_capsule, center=<6 0 0>, rotation=<rot 90 0 0>}]
 dyn_ear_l_base [{shapeType=Capsule, length=1, radius=1, shapeName=dyn_ear_l_base_capsule}]
 dyn_ear_r [{shapeType=Capsule, length=6, radius=3, shapeName=dyn_ear_r_capsule, center=<6 0 0>, rotation=<rot 90 0 0>}]
 dyn_ear_r_base [{shapeType=Capsule, length=1, radius=1, shapeName=dyn_ear_r_base_capsule}]
 dyn_root_curl_bk [{shapeType=Capsule, length=6, radius=3, shapeName=dyn_root_curl_bk_capsule, center=<6 0 -5>, rotation=<rot 10 0 0>}]
 dyn_root_curl_bk_l [{shapeType=Capsule, length=6, radius=3, shapeName=dyn_root_curl_bk_l_capsule, center=<6 0 -5>, rotation=<rot 10 0 0>}]
 dyn_root_curl_bk_r [{shapeType=Capsule, length=6, radius=4, shapeName=dyn_root_curl_bk_r_capsule, center=<6 0 -5>, rotation=<rot 10 0 0>}]
 head [
  {shapeType=Capsule, length=0, radius=9, shapeName=head_capsule, center=<4.6237 0 0>, rotation=<rot 90 0 0>}
  {shapeType=Capsule, length=1, radius=8, shapeName=head_capsule, center=<-0.7075 0 0>, rotation=<rot 90 0 0>}
  {shapeType=Capsule, length=10, radius=9, shapeName=head_capsule, center=<-2.7518 -11.6015 0>, rotation=<rot 0 179.9992 180>}
 ]
 neck_01 [{shapeType=Capsule, length=3.3374, radius=6.1629, shapeName=neck_01_capsule, rotation=<rot 90 0 0>}]
 neck_02 [{shapeType=Capsule, length=4.5914, radius=3.31, shapeName=neck_02_capsule, rotation=<rot 90 0 0>}]
 spine_05 [{shapeType=Capsule, length=20, radius=9.149, center=<5.8916 0.9595 0>}]
 upperarm_l [{shapeType=Capsule, length=21.75, radius=7, shapeName=upperarm_l_capsule, center=<12.3273 0 0>, rotation=<rot 90 0 -0.0002>}]
 upperarm_r [
  {shapeType=Capsule, length=21.75, radius=3.5, shapeName=upperarm_r_capsule, center=<-12.3273 0 0>, rotation=<rot 90 180 179.9995>}
  {
   shapeType=Capsule
   length=5
   radius=4.5
   shapeName=clavicle_r_capsule
   center=<-1.3476 -1.1017 -1.1338>
   rotation=<rot -59.9974 90.249 -0.286>
  }
 ]
}
constraints [
 {
  bone1Name=neck_01
  bone2Name=spine_05
  swing1LimitDegrees=45
  swing1Motion=Free
  swing2LimitDegrees=45
  swing2Motion=Free
  twistLimitDegrees=45
  twistMotion=Free
 }
 {
  bone1Name=neck_02
  bone2Name=neck_01
  swing1LimitDegrees=45
  swing1Motion=Free
  swing2LimitDegrees=45
  swing2Motion=Free
  twistLimitDegrees=45
  twistMotion=Free
 }
 {
  bone1Name=head
  bone2Name=neck_02
  swing1LimitDegrees=45
  swing1Motion=Free
  swing2LimitDegrees=45
  swing2Motion=Free
  twistLimitDegrees=45
  twistMotion=Free
 }
 {
  bone1Name=dyn_curl_fr_l_base
  bone2Name=head
  swing1LimitDegrees=45
  swing1Motion=Free
  swing2LimitDegrees=45
  swing2Motion=Free
  twistLimitDegrees=45
  twistMotion=Free
 }
 {bone1Name=dyn_curl_fr_l_01, bone2Name=dyn_curl_fr_l_base, swing1LimitDegrees=15, swing2LimitDegrees=30, twistLimitDegrees=5}
 {bone1Name=dyn_curl_fr_l_02, bone2Name=dyn_curl_fr_l_01, swing1LimitDegrees=15, swing2LimitDegrees=30, twistLimitDegrees=5}
 {
  bone1Name=dyn_curl_fr_r_base
  bone2Name=head
  swing1LimitDegrees=45
  swing1Motion=Free
  swing2LimitDegrees=45
  swing2Motion=Free
  twistLimitDegrees=45
  twistMotion=Free
 }
 {bone1Name=dyn_curl_fr_r_01, bone2Name=dyn_curl_fr_r_base, swing1LimitDegrees=10, swing2LimitDegrees=30, twistLimitDegrees=3}
 {bone1Name=dyn_curl_fr_r_02, bone2Name=dyn_curl_fr_r_01, swing1LimitDegrees=15, swing2LimitDegrees=45, twistLimitDegrees=5}
 {
  bone1Name=dyn_curl_fr_side_l_01_base
  bone2Name=head
  swing1LimitDegrees=45
  swing1Motion=Free
  swing2LimitDegrees=45
  swing2Motion=Free
  twistLimitDegrees=45
  twistMotion=Free
 }
 {bone1Name=dyn_curl_fr_side_l_01, bone2Name=dyn_curl_fr_side_l_01_base, swing1LimitDegrees=45, swing2LimitDegrees=45, twistLimitDegrees=5}
 {
  bone1Name=dyn_curl_fr_side_l_02_base
  bone2Name=head
  swing1LimitDegrees=45
  swing1Motion=Free
  swing2LimitDegrees=45
  swing2Motion=Free
  twistLimitDegrees=45
  twistMotion=Free
 }
 {bone1Name=dyn_curl_fr_side_l_02, bone2Name=dyn_curl_fr_side_l_02_base, swing1LimitDegrees=45, swing2LimitDegrees=45, twistLimitDegrees=5}
 {
  bone1Name=dyn_curl_fr_side_r_base
  bone2Name=head
  swing1LimitDegrees=45
  swing1Motion=Free
  swing2LimitDegrees=45
  swing2Motion=Free
  twistLimitDegrees=45
  twistMotion=Free
 }
 {bone1Name=dyn_curl_fr_side_r_01, bone2Name=dyn_curl_fr_side_r_base, swing1LimitDegrees=10, swing2LimitDegrees=30, twistLimitDegrees=3}
 {bone1Name=dyn_curl_fr_side_02, bone2Name=dyn_curl_fr_side_r_01, swing1LimitDegrees=15, swing2LimitDegrees=45, twistLimitDegrees=5}
 {
  bone1Name=dyn_curl_side_bk_l_base
  bone2Name=head
  swing1LimitDegrees=45
  swing1Motion=Free
  swing2LimitDegrees=45
  swing2Motion=Free
  twistLimitDegrees=45
  twistMotion=Free
 }
 {bone1Name=dyn_curl_side_bk_l, bone2Name=dyn_curl_side_bk_l_base, swing1LimitDegrees=30, swing2LimitDegrees=30, twistLimitDegrees=5}
 {
  bone1Name=dyn_curl_side_bk_r_base
  bone2Name=head
  swing1LimitDegrees=45
  swing1Motion=Locked
  swing2LimitDegrees=45
  swing2Motion=Locked
  twistLimitDegrees=45
  twistMotion=Locked
 }
 {bone1Name=dyn_curl_side_bk_r, bone2Name=dyn_curl_side_bk_r_base, swing1LimitDegrees=15, swing2LimitDegrees=15, twistLimitDegrees=5}
 {
  bone1Name=dyn_curl_side_l_base
  bone2Name=head
  swing1LimitDegrees=45
  swing1Motion=Free
  swing2LimitDegrees=45
  swing2Motion=Free
  twistLimitDegrees=45
  twistMotion=Free
 }
 {bone1Name=dyn_curl_side_l_01, bone2Name=dyn_curl_side_l_base, swing1LimitDegrees=5, swing2LimitDegrees=15, twistLimitDegrees=5}
 {bone1Name=dyn_curl_side_l_02, bone2Name=dyn_curl_side_l_01, swing1LimitDegrees=15, swing2LimitDegrees=30, twistLimitDegrees=5}
 {
  bone1Name=dyn_curl_side_r_base
  bone2Name=head
  swing1LimitDegrees=45
  swing1Motion=Free
  swing2LimitDegrees=45
  swing2Motion=Free
  twistLimitDegrees=45
  twistMotion=Free
 }
 {bone1Name=dyn_curl_side_r_01, bone2Name=dyn_curl_side_r_base, swing1LimitDegrees=10, swing2LimitDegrees=30, twistLimitDegrees=3}
 {bone1Name=dyn_curl_side_r_02, bone2Name=dyn_curl_side_r_01, swing1LimitDegrees=15, swing2LimitDegrees=45, twistLimitDegrees=5}
 {
  bone1Name=dyn_ear_l_base
  bone2Name=head
  swing1LimitDegrees=45
  swing1Motion=Free
  swing2LimitDegrees=45
  swing2Motion=Free
  twistLimitDegrees=45
  twistMotion=Free
 }
 {bone1Name=dyn_ear_l, bone2Name=dyn_ear_l_base, swing1LimitDegrees=45, swing2LimitDegrees=45, twistLimitDegrees=5}
 {
  bone1Name=dyn_ear_r_base
  bone2Name=head
  swing1LimitDegrees=45
  swing1Motion=Locked
  swing2LimitDegrees=45
  swing2Motion=Locked
  twistLimitDegrees=45
  twistMotion=Locked
 }
 {bone1Name=dyn_ear_r, bone2Name=dyn_ear_r_base, swing1LimitDegrees=30, swing2LimitDegrees=45, twistLimitDegrees=5}
 {bone1Name=dyn_root_curl_bk, bone2Name=head, swing1LimitDegrees=30, swing2LimitDegrees=30, twistLimitDegrees=5}
 {bone1Name=dyn_curl_bk_01, bone2Name=dyn_root_curl_bk_l, swing1LimitDegrees=30, swing2LimitDegrees=30, twistLimitDegrees=5}
 {bone1Name=dyn_curl_bk_02, bone2Name=dyn_root_curl_bk_l, swing1LimitDegrees=30, swing2LimitDegrees=30, twistLimitDegrees=5}
 {bone1Name=dyn_curl_bk_03, bone2Name=dyn_root_curl_bk, swing1LimitDegrees=30, swing2LimitDegrees=30, twistLimitDegrees=5}
 {bone1Name=dyn_curl_bk_04, bone2Name=dyn_root_curl_bk, swing1LimitDegrees=30, swing2LimitDegrees=30, twistLimitDegrees=5}
 {bone1Name=dyn_curl_bk_05, bone2Name=dyn_root_curl_bk, swing1LimitDegrees=30, swing2LimitDegrees=30, twistLimitDegrees=5}
 {bone1Name=dyn_curl_bk_06, bone2Name=dyn_root_curl_bk, swing1LimitDegrees=30, swing2LimitDegrees=30, twistLimitDegrees=5}
 {bone1Name=dyn_curl_bk_07, bone2Name=dyn_root_curl_bk_r, swing1LimitDegrees=30, swing2LimitDegrees=30, twistLimitDegrees=5}
 {bone1Name=dyn_curl_bk_08, bone2Name=dyn_root_curl_bk_r, swing1LimitDegrees=30, swing2LimitDegrees=30, twistLimitDegrees=5}
 {bone1Name=dyn_curl_bk_09, bone2Name=dyn_root_curl_bk_r, swing1LimitDegrees=30, swing2LimitDegrees=30, twistLimitDegrees=5}
 {
  bone1Name=clavicle_l
  bone2Name=spine_05
  swing1LimitDegrees=45
  swing1Motion=Free
  swing2LimitDegrees=45
  swing2Motion=Free
  twistLimitDegrees=45
  twistMotion=Free
 }
 {bone1Name=dyn_root_curl_bk_l, bone2Name=head, swing1LimitDegrees=30, swing2LimitDegrees=30, twistLimitDegrees=5}
 {bone1Name=dyn_root_curl_bk_r, bone2Name=head, swing1LimitDegrees=10, swing2LimitDegrees=30, twistLimitDegrees=5}
]
